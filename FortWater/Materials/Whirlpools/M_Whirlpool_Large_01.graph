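# Masked two-sided Single Layer Water material for a large whirlpool that blends into surrounding Gerstner waves.
# Radial-falloff rotation twists UVs into scrolling polar coordinates, driving blurred turbulence displacement, distorted foam and under-surface scattering.
# Normal maps are re-oriented along the radial direction and faded into wave normals toward the rim, where a dithered opacity mask trims the edge.
# Low and Medium quality drop foam UV distortion and the normal blend.
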Material /FortWater/Materials/Whirlpools/M_Whirlpool_Large_01
graph {
 properties {
  bTangentSpaceNormal=false
  bUseMaterialAttributes=true
  bUsedWithWater=true
  blendMode=BLEND_Masked
  pixelDepthOffsetMode=PDOM_Legacy
  shadingModel=MSM_SingleLayerWater
  twoSided=true
  usedShadingModels=SingleLayerWater
 }
 nodes [
  n1 = MaterialFunctionCall!(
   Z Offset Only
   Position
   WaterBodyIndex=ScalarParameter() {defaultValue=14, parameterName=Index}
   Normals use VS Interpolation=StaticBool()
   All Normals VS
  ) {materialFunction=/FortWater/Materials/Functions/Fort_ComputeGerstnerWaves}
  n2 = MaterialFunctionCall() {materialFunction=&VectorToRadialValue}
  n3 = OneMinus(None=Saturate(None=Multiply(A=@n2.Linear Distance)))
  n4 = ScalarParameter() {defaultValue=3, parameterName=WhirlPower}
  n5 = AppendVector(
   A=MaterialFunctionCall!(X range Y range=Constant4Vector() {constant=<rgba -1 1 -1 1>}.RGBA, UVs) {materialFunction=/Engine/Functions/Engine_MaterialFunctions01/Coordinates/UVRemap_0-1_ToRange}.UV
   B=Constant()
  )
  n6 = RotateAboutAxis(
   NormalizedRotationAxis=Constant3Vector() {constant=<rgba 0 0 1 0>}
   RotationAngle=Multiply(A=ScalarParameter() {defaultValue=0.6, parameterName=Rad}, B=Power(Base=@n3, Exp=@n4))
   PivotPoint=Constant()
   Position=@n5
  )
  n7 = MaterialFunctionCall(Float2=MaterialFunctionCall(Vector or UVs=Add(A=@n6, B=@n5)) {materialFunction=&VectorToRadialValue}.Radial Coordinates) {materialFunction=&BreakOutFloat2Components}
  n8 = ScalarParameter() {defaultValue=1, parameterName=Y Power}
  n9 = AppendVector(A=AppendVector(A=@n7.R, B=Power(Base=@n7.G, Exp=@n8)), B=Constant())
  n10 = Constant3Vector() {constant=<rgba 0 0.2 0 0>}
  n11 = Multiply(
   A=MaterialFunctionCall(Speed=ScalarParameter() {defaultValue=0.075, parameterName=Speed}) {materialFunction=&TimeWithSpeedVariable}.Time
   B=@n10
  )
  n12 = ComponentMask(None=Multiply(A=Add(A=@n9, B=@n11), B=VectorParameter() {parameterName=Whirl Tiling, defaultValue=<rgba 3 2 0 0>}.RGB)) {g=true, r=true}
  n13 = TextureSample!(UVs=@n12, Tex, Level=ScalarParameter() {defaultValue=6.5, parameterName=Foam Mip Scat}, Apply View MipBias) {mipValueMode=TMVM_MipLevel, texture=/FortWater/Textures/T_Turbulence_Seamless}
  n14 = Multiply(
   A=ScalarParameter() {defaultValue=96, parameterName=WPO}
   B=Subtract(A=@n13.R, B=ScalarParameter() {defaultValue=0.15, parameterName=WPO Offset}) {constB=0.4}
  )
  n15 = Constant3Vector() {constant=<rgba 0 0 1 0>}
  n16 = OneMinus(None=Multiply(A=MaterialFunctionCall() {materialFunction=&VectorToRadialValue}.Linear Distance))
  Add(
   A=Multiply(A=@n1.WPO, B=ScalarParameter() {defaultValue=1, parameterName=WaveWPO})
   B=Multiply(A=Multiply(A=@n14, B=@n15), B=Saturate(None=@n16))
  )
  ScalarParameter() {group=Water Shading Model, parameterName=Surface Opacity}
  ScalarParameter() {defaultValue=0.5, parameterName=A}
  n17 = Multiply(A=MaterialFunctionCall() {materialFunction=&VectorToRadialValue}.Linear Distance)
  n18 = Multiply(
   A=Subtract(A=OneMinus(None=@n17), B=ScalarParameter() {defaultValue=0.5, parameterName=Foam Bias})
   B=ScalarParameter() {defaultValue=0.3, parameterName=Foam Center}
  )
  n19 = Multiply(A=ScalarParameter() {defaultValue=8, parameterName=Foam Breakup Tiling}, B=@n12)
  n20 = ComponentMask(None=Multiply(
    A=TextureSample!(UVs=@n19, Tex, Apply View MipBias) {samplerType=SAMPLERTYPE_Normal, texture=/Game/Global/Textures/Default/Tileable_Patterns/T_TRV_Moss_01_N}.RGB
    B=ScalarParameter() {defaultValue=0.05, parameterName=Foam UV2 Distortion}
   )) {g=true, r=true}
  n21 = Constant()
  n22 = QualitySwitch(Default=@n20, Low=@n21, High=@n20, Medium=@n21)
  n23 = TextureCoordinate()
  n24 = ScalarParameter() {defaultValue=0.125, parameterName=Speed 2}
  n25 = Add(
   A=@n9
   B=Multiply(A=MaterialFunctionCall(Speed=@n24) {materialFunction=&TimeWithSpeedVariable}.Time, B=@n10)
  )
  n26 = VectorParameter() {parameterName=Whirl Tiling 2, defaultValue=<rgba 10 6 0 0>}
  n27 = TextureCoordinate()
  n28 = DDX(Value=@n27)
  n29 = DDY(Value=@n27)
  n30 = TextureSample!(UVs=ComponentMask(None=Multiply(A=@n25, B=@n26.RGB)) {g=true, r=true}, Tex, "DDX(UVs)"=@n28, "DDY(UVs)"=@n29, Apply View MipBias) {mipValueMode=TMVM_Derivative, samplerType=SAMPLERTYPE_Normal, texture=/FortWater/Textures/T_Water_TilingNormal_Waves_01}
  n31 = ComponentMask(None=Multiply(A=@n30.RGB, B=ScalarParameter() {defaultValue=-0.05, parameterName=Foam UV Distortion})) {g=true, r=true}
  n32 = Constant()
  n33 = Add(A=@n12, B=QualitySwitch(Default=@n31, Low=@n32, High=@n31, Medium=@n32))
  n34 = TextureCoordinate()
  n35 = Add(
   A=TextureSample(UVs=Add(A=@n12, B=@n22), "DDX(UVs)"=DDX(Value=@n23), "DDY(UVs)"=DDY(Value=@n23)) {mipValueMode=TMVM_Derivative, texture=&T_Turbulence_Seamless}.R
   B=TextureSample(UVs=@n33, "DDX(UVs)"=DDX(Value=@n34), "DDY(UVs)"=DDY(Value=@n34)) {mipValueMode=TMVM_Derivative, texture=&T_Turbulence_Seamless}.R
  )
  n36 = ScalarParameter() {defaultValue=0.2, parameterName=Foam Texture}
  n37 = Multiply(A=Add(A=@n18, B=Multiply(A=@n35, B=@n36)), B=ScalarParameter() {defaultValue=0.5, parameterName=Foam Amount})
  n38 = Normalize(VectorInput=Subtract(A=TextureCoordinate()) {constB=0.5})
  n39 = MaterialFunctionCall(Float2=@n38) {materialFunction=&BreakOutFloat2Components}
  n40 = TextureSample(UVs=@n33, "DDX(UVs)"=@n28, "DDY(UVs)"=@n29) {mipValueMode=TMVM_Derivative, samplerType=SAMPLERTYPE_Normal, texture=/FortWater/Textures/T_Turbulence_Seamless_N}
  n41 = ComponentMask(None=Add(A=@n40.RGB, B=@n30.RGB)) {g=true, r=true}
  n42 = Multiply(A=AppendVector(A=Multiply(A=@n39.G) {constB=-1}, B=@n39.R), B=@n41)
  n43 = Multiply(A=@n38, B=@n41)
  n44 = MaterialFunctionCall(
   Normal=AppendVector(A=Add(A=@n42, B=@n43), B=Multiply(A=@n40.B, B=@n30.B))
   Flatness=ScalarParameter() {defaultValue=0.75, parameterName=Flatten}
  ) {materialFunction=&FlattenNormal}
  n45 = Subtract(
   A=OneMinus(None=MaterialFunctionCall() {materialFunction=&VectorToRadialValue}.Linear Distance)
   B=ScalarParameter() {parameterName=Outer Alpha Bias}
  )
  n46 = ScalarParameter() {defaultValue=1, parameterName=Outer Alpha Width}
  n47 = LinearInterpolate(A=@n1.Normal, B=@n44.Result, Alpha=Saturate(None=Divide(A=@n45, B=@n46)))
  n48 = Add(
   A=OneMinus(None=MaterialFunctionCall() {materialFunction=&VectorToRadialValue}.Linear Distance)
   B=ScalarParameter() {parameterName=asdf}
  )
  n49 = ScalarParameter() {defaultValue=8, parameterName=Edge Sharpness}
  SetMaterialAttributes!(
   MaterialAttributes
   Base Color=VectorParameter() {parameterName=BaseColor, defaultValue=<rgba 0.7 0.9 1 0.1>}.RGB
   Specular=ScalarParameter() {defaultValue=0.225, group=Water Shading Model, parameterName=Specular}
   Roughness=ScalarParameter() {defaultValue=0.03, group=Water Shading Model, parameterName=Roughness}
   Opacity=Add(A=Saturate(None=@n37), B=ScalarParameter() {parameterName=Opacity})
   Normal=QualitySwitch(Default=@n47, Low=@n44.Result, High=@n47, Medium=@n44.Result)
   World Position Offset
   Emissive Color
   Opacity Mask=MaterialFunctionCall(Alpha Threshold=Multiply(A=@n48, B=@n49)) {materialFunction=&DitherTemporalAA}.Result
  ) {attributeSetTypes [
    #1
    #2
    #3
    #4
    #5
    #6
    #7
    #8
   ]}
  n50 = VectorParameter() {group=Water Shading Model, parameterName=Scattering, defaultValue=<rgba 0.1 0.1 0.1 0.5>}
  n51 = Divide(A=@n50.A) {constB=100}
  n52 = ScalarParameter() {defaultValue=8, parameterName=Foam Center Under Scatter}
  n53 = ScalarParameter() {defaultValue=4, parameterName=Foam Tex Scat}
  n54 = Saturate(None=Add(A=Multiply(A=@n18, B=@n52), B=Multiply(A=@n13.R, B=@n53)))
  n55 = VectorParameter() {group=Water Shading Model, parameterName=Foam Under Scattering, defaultValue=<rgba 1 1 1 0>}
  n56 = Multiply(A=@n55.RGB, B=Divide(A=@n55.A) {constB=100})
  n57 = VectorParameter() {group=Water Shading Model, parameterName=Absorption, defaultValue=<rgba 1 8 20 100>}
  SingleLayerWaterMaterialOutput!(
   ScatteringCoefficients=Add(A=Multiply(A=@n50.RGB, B=@n51), B=Multiply(A=@n54, B=@n56))
   AbsorptionCoefficients=Divide(B=Multiply(A=@n57.RGB, B=@n57.RGB))
   PhaseG=CollectionParameter() {collection=/Water/Materials/MPC/MPC_Water, parameterName=Ansotropy}
   ColorScaleBehindWater
  )
 ]
}
