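# Masked, two-sided Single Layer Water material for a small whirlpool, with a dithered outer edge.
# Center-twisted radial UVs pan over time to sample turbulence and normal maps, driving foam in opacity and water scattering.
# Swirl normals blend over Gerstner ocean normals, except on Low and Medium quality.
# World position offset combines waves, turbulence displacement, a distance-faded Z bias and a downward pull when the camera is below.
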Material /FortWater/Materials/Whirlpools/M_Whirlpool_Small_01
graph {
 properties {
  bTangentSpaceNormal=false
  bUseMaterialAttributes=true
  bUsedWithNanite=true
  blendMode=BLEND_Masked
  pixelDepthOffsetMode=PDOM_Legacy
  refractionMethod=RM_PixelNormalOffset
  shadingModel=MSM_SingleLayerWater
  twoSided=true
  usedShadingModels=SingleLayerWater
 }
 nodes [
  n1 = OneMinus(None=MaterialFunctionCall() {materialFunction=&VectorToRadialValue}.Linear Distance)
  n2 = ScalarParameter() {defaultValue=0.2, parameterName=Foam Alpha Edge}
  n3 = Saturate(None=Divide(A=Subtract(A=@n1, B=@n2), B=ScalarParameter() {defaultValue=0.2, parameterName=Foam Alpha Width}))
  n4 = Multiply(A=MaterialFunctionCall() {materialFunction=&VectorToRadialValue}.Linear Distance)
  n5 = Multiply(
   A=Subtract(A=OneMinus(None=@n4), B=ScalarParameter() {defaultValue=0.6, parameterName=Foam Bias})
   B=ScalarParameter() {defaultValue=0.2, parameterName=Foam Center}
  )
  n6 = MaterialFunctionCall() {materialFunction=&VectorToRadialValue}
  n7 = OneMinus(None=Saturate(None=Multiply(A=@n6.Linear Distance)))
  n8 = ScalarParameter() {defaultValue=2, parameterName=WhirlPower}
  n9 = AppendVector(
   A=MaterialFunctionCall!(X range Y range=Constant4Vector() {constant=<rgba -1 1 -1 1>}.RGBA, UVs) {materialFunction=/Engine/Functions/Engine_MaterialFunctions01/Coordinates/UVRemap_0-1_ToRange}.UV
   B=Constant()
  )
  n10 = RotateAboutAxis(
   NormalizedRotationAxis=Constant3Vector() {constant=<rgba 0 0 1 0>}
   RotationAngle=Multiply(A=ScalarParameter() {defaultValue=0.6, parameterName=Rad}, B=Power(Base=@n7, Exp=@n8))
   PivotPoint=Constant()
   Position=@n9
  )
  n11 = AppendVector(
   A=MaterialFunctionCall(Vector or UVs=Add(A=@n10, B=@n9)) {materialFunction=&VectorToRadialValue}.Radial Coordinates
   B=Constant()
  )
  n12 = VectorParameter() {parameterName=Whirl Dirvish, defaultValue=<rgba 0 0.2 0 0>}
  n13 = Multiply(
   A=MaterialFunctionCall(Speed=ScalarParameter() {defaultValue=1, parameterName=Speed}) {materialFunction=&TimeWithSpeedVariable}.Time
   B=@n12.RGB
  )
  n14 = ComponentMask(None=Multiply(A=Add(A=@n11, B=@n13), B=VectorParameter() {parameterName=Whirl Tiling, defaultValue=<rgba 3 2 0 0>}.RGB)) {g=true, r=true}
  n15 = TextureCoordinate()
  n16 = Multiply(
   A=TextureSample!(UVs=@n14, Tex, "DDX(UVs)"=DDX(Value=@n15), "DDY(UVs)"=DDY(Value=@n15), Apply View MipBias) {mipValueMode=TMVM_Derivative, texture=/FortWater/Textures/T_Turbulence_Seamless}.R
   B=ScalarParameter() {defaultValue=0.5, parameterName=Foam Texture}
  )
  n17 = Saturate(None=Multiply(A=Add(A=@n5, B=@n16), B=ScalarParameter() {defaultValue=0.5, parameterName=Foam Amount}))
  n18 = MaterialFunctionCall!(
   Z Offset Only
   Position
   WaterBodyIndex=CollectionParameter() {collection=/Water/Materials/MPC/MPC_Water, parameterName=OceanWaterBodyIndex}
   Normals use VS Interpolation
   All Normals VS
  ) {materialFunction=/FortWater/Materials/Functions/Fort_ComputeGerstnerWaves}
  n19 = Normalize(VectorInput=Subtract(A=TextureCoordinate()) {constB=0.5})
  n20 = MaterialFunctionCall(Float2=@n19) {materialFunction=&BreakOutFloat2Components}
  n21 = Multiply(A=@n20.G) {constB=-1}
  n22 = TextureCoordinate()
  n23 = DDX(Value=@n22)
  n24 = DDY(Value=@n22)
  n25 = TextureSample(UVs=@n14, "DDX(UVs)"=@n23, "DDY(UVs)"=@n24) {mipValueMode=TMVM_Derivative, samplerType=SAMPLERTYPE_Normal, texture=/FortWater/Textures/T_Turbulence_Seamless_N}
  n26 = ScalarParameter() {defaultValue=2, parameterName=Speed 2}
  n27 = Add(
   A=@n11
   B=Multiply(
    A=MaterialFunctionCall(Speed=@n26) {materialFunction=&TimeWithSpeedVariable}.Time
    B=@n12.RGB
   )
  )
  n28 = VectorParameter() {parameterName=Whirl Tiling 2, defaultValue=<rgba 2 1 0 0>}
  n29 = TextureSample(UVs=ComponentMask(None=Multiply(A=@n27, B=@n28.RGB)) {g=true, r=true}, "DDX(UVs)"=@n23, "DDY(UVs)"=@n24) {mipValueMode=TMVM_Derivative, samplerType=SAMPLERTYPE_Normal, texture=/FortWater/Textures/T_Water_TilingNormal_Waves_01}
  n30 = ComponentMask(None=Add(A=@n25.RGB, B=@n29.RGB)) {g=true, r=true}
  n31 = Add(A=Multiply(A=AppendVector(A=@n21, B=@n20.R), B=@n30), B=Multiply(A=@n19, B=@n30))
  n32 = Multiply(A=@n25.B, B=@n29.B)
  n33 = LinearInterpolate(
   A=@n18.Normal
   B=MaterialFunctionCall(Normal=AppendVector(A=@n31, B=@n32), Flatness=ScalarParameter() {defaultValue=0.75, parameterName=Flatten}) {materialFunction=&FlattenNormal}.Result
   Alpha=@n3
  )
  n34 = ScalarParameter() {defaultValue=1, parameterName=WaveWPO}
  n35 = ScalarParameter() {defaultValue=256, parameterName=WPO}
  n36 = TextureSample!(UVs=@n14, Tex, Level=ScalarParameter() {defaultValue=7.5, parameterName=Foam Mip Scat}, Apply View MipBias) {mipValueMode=TMVM_MipLevel, texture=&T_Turbulence_Seamless}
  n37 = Subtract(A=@n36.R, B=ScalarParameter() {defaultValue=0.05, parameterName=WPO Offset}) {constB=0.4}
  n38 = Multiply(A=MaterialFunctionCall() {materialFunction=&VectorToRadialValue}.Linear Distance)
  n39 = Multiply(A=Multiply(A=Multiply(A=@n35, B=@n37), B=Constant3Vector() {constant=<rgba 0 0 1 0>}), B=Saturate(None=OneMinus(None=@n38)))
  n40 = ScalarParameter() {defaultValue=8000, parameterName=Fade Length and Offset}
  n41 = Multiply(
   A=MaterialFunctionCall(Fade Length=@n40, Fade Offset=@n40) {materialFunction=&CameraDepthFade}.Result
   B=Multiply(A=Constant3Vector() {constant=<rgba 0 0 1 0>}, B=ScalarParameter() {defaultValue=128, parameterName=Distance Z Bias})
  )
  n42 = MaterialFunctionCall() {materialFunction=&VectorToRadialValue}
  n43 = Multiply(
   A=Add(A=OneMinus(None=@n42.Linear Distance), B=ScalarParameter() {defaultValue=0.05, parameterName=Outer Dither Bias})
   B=ScalarParameter() {defaultValue=1, parameterName=Outer Dither Amt}
  )
  n44 = MaterialFunctionCall() {materialFunction=&DitherTemporalAA}
  n45 = ScalarParameter() {defaultValue=0.4, parameterName=Alpha Multiply}
  n46 = SetMaterialAttributes!(
   MaterialAttributes
   Base Color=VectorParameter() {parameterName=BaseColor, defaultValue=<rgba 0.7 0.9 1 0.1>}.RGB
   Specular=ScalarParameter() {defaultValue=0.225, group=Water Shading Model, parameterName=Specular}
   Roughness=ScalarParameter() {defaultValue=0.03, group=Water Shading Model, parameterName=Roughness}
   Opacity=Add(A=Multiply(A=@n3, B=@n17), B=ScalarParameter() {parameterName=Opacity})
   Normal=QualitySwitch(Default=@n33, Low=@n18.Normal, High=@n33, Medium=@n18.Normal)
   World Position Offset=Add(A=Multiply(A=@n18.WPO, B=@n34), B=Add(A=@n39, B=@n41))
   Emissive Color
   Opacity Mask=Add(A=Saturate(None=@n43), B=Multiply(A=@n44.Result, B=@n45))
  ) {attributeSetTypes [
    #1
    #2
    #3
    #4
    #5
    #6
    #7
    #8
   ]}
  n47 = Subtract(A=ActorPositionWS(), B=CameraPositionWS())
  n48 = Saturate(None=Divide(A=ComponentMask(None=@n47) {b=true}, B=ScalarParameter() {defaultValue=16, parameterName=Underside Bias Length}))
  n49 = ScalarParameter() {defaultValue=32, parameterName=Camera Pull}
  MaterialFunctionCall!(
   Attributes=MaterialFunctionCall!(Attributes=@n46, Depth) {materialFunction=/FortWater/Materials/Functions/TwoSidedWaterCorrections}.Result
   Additional WPO=Multiply(A=Constant3Vector() {constant=<rgba 0 0 -1 0>}, B=Multiply(A=@n48, B=@n49))
  ) {materialFunction=/FortWater/Materials/Functions/AttributesAddWPO}
  Normalize(VectorInput=Subtract(A=WorldPosition().XYZ, B=CameraPositionWS()))
  ScalarParameter() {defaultValue=15, parameterName=Index}
  ScalarParameter() {defaultValue=0.5, parameterName=A}
  n50 = VectorParameter() {group=Water Shading Model, parameterName=Scattering, defaultValue=<rgba 0.1 0.1 0.1 0.5>}
  n51 = Divide(A=@n50.A) {constB=100}
  MaterialExpressionAdd_5 = Add(
   A=Multiply(A=@n5, B=ScalarParameter() {defaultValue=8, parameterName=Foam Center Under Scatter})
   B=Multiply(A=@n36.R, B=ScalarParameter() {defaultValue=6, parameterName=Foam Tex Scat})
  )
  n52 = Custom!(x=@MaterialExpressionAdd_5) {
   code="return 1-exp(-x);"
   description=exp
   outputType=CMOT_Float1
   inputs [{inputName=x, input {expression=&self:MaterialExpressionAdd_5}}]
  }
  n53 = Multiply(A=@n3, B=Saturate(None=@n52))
  n54 = VectorParameter() {group=Water Shading Model, parameterName=Foam Under Scattering, defaultValue=<rgba 1 1 1 0.2>}
  n55 = Multiply(A=@n54.RGB, B=Divide(A=@n54.A) {constB=100})
  n56 = VectorParameter() {group=Water Shading Model, parameterName=Absorption, defaultValue=<rgba 1 8 20 100>}
  SingleLayerWaterMaterialOutput!(
   ScatteringCoefficients=Add(A=Multiply(A=@n50.RGB, B=@n51), B=Multiply(A=@n53, B=@n55))
   AbsorptionCoefficients=Divide(B=Multiply(A=@n56.RGB, B=@n56.RGB))
   PhaseG=CollectionParameter() {collection=&MPC_Water, parameterName=Ansotropy}
   ColorScaleBehindWater
  )
  StaticBool()
 ]
}
